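; Base Blueprint for a player-placed world marker in Athena, showing meshes by marker type and tinting them to the marker color.
; Setup shows the main mesh for Location and the ping proto for Location or Enemy, creating main and target dynamic materials here because construction-script creation rendered unreliably.
; Color changes tint the materials, forcing red on the experimental proto for Enemy.
; Placement updates the material pivot, resets spawn parameters and starts the spawn animation.
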(event EventOnSetupMarker
  (params (MarkerData :type Fort%20World%20Marker%20Data%20Structure%20%28by%20ref%29))
  (bind (_markerid _owner _markertype) (Utilities|Struct|BreakFortWorldMarkerData MarkerData))
  (Variables|Default|SetMarkerData MarkerData)
  (Rendering|SetVisibility (Variables|FortPlayerMarkerBase|GetStaticMeshComponent) (Utilities|Enum|Equal(Enum) _markertype "Location") true)
  (Rendering|SetVisibility (Variables|Default|GetPingMarkerProto) (Utilities|Select :Index _markertype :None false :Location true :Enemy true :Item false :Elimination false :SpecialLocal false :SpecialServer false :Teammate false :PingGoTo false :PingHoldPosition false :PingNoIcon false :PingAttack false :SquadComsWatchArea false :SquadComsDefend false :SquadComsSomeonesHere false :SquadComsLootArea false :SquadComsAttack false :SquadComsFallBack false :SquadRequestsAmmoLight false :SquadRequestsAmmoMedium false :SquadRequestsAmmoHeavy false :SquadRequestsAmmoShells false :SquadRequestsAmmoExplosives false :SquadRequestsMaterialWood false :SquadRequestsMaterialStone false :SquadRequestsMaterialMetal false :SquadRequestsHealth false :SquadRequestsShields false :SquadRequestsGold false :SquadComsHelp false :SquadComsPickMeUp false :SquadComsThankYou false :SquadComsReady false :SquadComsFollowMe false :SquadComsHoldOn false) false)
  ;[1 1 1] This was previously in the construction script, but moved here due to issues of either not showing up or rendering the wrong color
  (bind _returnvalue (Rendering|Material|CreateDynamicMaterialInstance (Variables|FortPlayerMarkerBase|GetStaticMeshComponent) 0 0 "None" :!class /Script/Engine.PrimitiveComponent))
  (Variables|FortPlayerMarkerBase|SetMID_Main _returnvalue)
  (bind _returnvalue_2 (Rendering|Material|CreateDynamicMaterialInstance (Variables|Default|GetSM-Target) 0 0 "None" :!class /Script/Engine.PrimitiveComponent))
  (Variables|FortPlayerMarkerBase|SetMID_Target _returnvalue_2)
  (Rendering|Material|SetVectorParameterValue (Variables|FortPlayerMarkerBase|GetMID_Main) "ObjectPivotLoc" (Math|Conversions|ToLinearColor(Vector) (Transformation|GetWorldLocation (Variables|FortPlayerMarkerBase|GetStaticMeshComponent))) :!class /Script/Engine.MaterialInstanceDynamic))

(event EventOnMarkerColorChanged
  (params (InColor :type Linear%20Color%20Structure))
  (bind (_markerid _owner _markertype) (Utilities|Struct|BreakFortWorldMarkerData (Variables|Default|GetMarkerData)))
  (bind _midprotomarker (Variables|Default|GetMIDProtoMarker))
  (bind _midprotomarker_2 (Variables|Default|GetMIDProtoMarker))
  (bind _mid_main (Variables|FortPlayerMarkerBase|GetMID_Main))
  (Rendering|Material|SetVectorParameterValue (Variables|FortPlayerMarkerBase|GetMID_Main) "MainColor" InColor :!class /Script/Engine.MaterialInstanceDynamic)
  (Rendering|Material|SetScalarParameterValue _mid_main "SpawnOpacity" 0.0 :!class /Script/Engine.MaterialInstanceDynamic)
  (Rendering|Material|SetScalarParameterValue (Variables|FortPlayerMarkerBase|GetMID_Target) "SpawnScale" 2.5 :!class /Script/Engine.MaterialInstanceDynamic)
  (Rendering|Material|SetScalarParameterValue _mid_main "SpawnScaleVertical" 0.0 :!class /Script/Engine.MaterialInstanceDynamic)
  ;[0 0 0] ProtoMarker expirementing
  (Utilities|IsValid _midprotomarker
    (:Is%20Valid
      (if (Utilities|Enum|Equal(Enum) _markertype "Enemy")
        (Rendering|Material|SetVectorParameterValue _midprotomarker_2 "MainColor" "(R=0.706597,G=0.000000,B=0.000000,A=1.000000)" :!class /Script/Engine.MaterialInstanceDynamic)
        (Rendering|Material|SetVectorParameterValue _midprotomarker_2 "BaseColor" "(R=1.000000,G=0.000000,B=0.093845,A=1.000000)" :!class /Script/Engine.MaterialInstanceDynamic)
        (else
          (Rendering|Material|SetVectorParameterValue _midprotomarker "MainColor" (Math|Color|LinearColor*Float InColor 0.6) :!class /Script/Engine.MaterialInstanceDynamic)
          (Rendering|Material|SetVectorParameterValue _midprotomarker "BaseColor" InColor :!class /Script/Engine.MaterialInstanceDynamic))))))

(event EventOnMarkerPlaced
  (bind _mid_main (Variables|FortPlayerMarkerBase|GetMID_Main))
  (Rendering|Material|SetVectorParameterValue _mid_main "ObjectPivotLoc" (Math|Conversions|ToLinearColor(Vector) (Transformation|GetWorldLocation (Variables|FortPlayerMarkerBase|GetStaticMeshComponent))) :!class /Script/Engine.MaterialInstanceDynamic)
  (Rendering|Material|SetScalarParameterValue _mid_main "SpawnOpacity" 0.0 :!class /Script/Engine.MaterialInstanceDynamic)
  (Rendering|Material|SetScalarParameterValue (Variables|FortPlayerMarkerBase|GetMID_Target) "SpawnScale" 2.5 :!class /Script/Engine.MaterialInstanceDynamic)
  (Rendering|Material|SetScalarParameterValue _mid_main "SpawnScaleVertical" 0.0 :!class /Script/Engine.MaterialInstanceDynamic)
  (CallFunction|StartSpawnMarkerAnim))

(fn ConstructionScript
  (bind _returnvalue (Rendering|Material|CreateDynamicMaterialInstance (Variables|Default|GetPingMarkerProto) 0 0 "None" :!class /Script/Engine.PrimitiveComponent))
  (Variables|Default|SetMIDProtoMarker _returnvalue))
