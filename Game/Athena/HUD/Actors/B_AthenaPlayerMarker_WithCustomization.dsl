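; B_AthenaPlayerMarker_WithCustomization is a player marker actor Blueprint for the Athena HUD.
; Its setup, hidden and placed marker events, along with its construction script, only forward to the parent implementation.
; The graphs add no logic of their own, so any marker or customization behavior comes from outside them.
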(event EventOnSetupMarker
  (params (MarkerData :type Fort%20World%20Marker%20Data%20Structure%20%28by%20ref%29))
  (|CallParentFunction MarkerData))

(event EventOnMarkerHidden
  (|CallParentFunction))

(event EventOnMarkerPlaced
  (|CallParentFunction))

(fn ConstructionScript
  (|CallParentFunction))
