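# Translucent UI material that draws a boss health bar bracket with top and bottom spike accents, built procedurally without textures.
# Perspective-warped SDF polygons and a trapezoid are subtracted and antialiased into bracket and spike masks.
# Spike_Offset shifts the spikes, and SpikesOpacities scales a combined mask term that is maxed with another.
# IsLocked blends light gray toward near-black tones, and Attacked blends that toward a mask-driven mix of red and the Color parameter.
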Material /Game/Athena/HUD/Actors/Materials/M_UI_BossHealthBar_BracketWithAccent
graph {
 properties {bAutomaticallySetUsageInEditor=false, blendMode=BLEND_Translucent, materialDomain=MD_UI, usedShadingModels=Default Lit}
 nodes [
  Constant2Vector() {g=0.22, r=-0.12}
  n1 = Constant3Vector() {constant=<rgba 0.0156 0.0156 0.0156 0>}
  n2 = Constant3Vector() {constant=<rgba 0.0312 0.0312 0.0312 0>}
  n3 = NamedRerouteUsage()
  LinearInterpolate(
   A=LinearInterpolate(
    A=Constant3Vector() {constant=<rgba 0.5333 0.5333 0.5333 1>}
    B=LinearInterpolate(A=@n1, B=@n2, Alpha=@n3)
    Alpha=ScalarParameter() {parameterName=IsLocked}
   ) {constB=0}
   B=LinearInterpolate(
    A=Constant3Vector() {constant=<rgba 1 0.013 0.0931 1>}
    B=VectorParameter() {parameterName=Color, defaultValue=<rgba 0.9301 1 0.0103 1>}.RGB
    Alpha=NamedRerouteUsage()
   )
   Alpha=ScalarParameter() {parameterName=Attacked}
  )
  n4 = NamedRerouteUsage()
  n5 = NamedRerouteUsage()
  Max(A=NamedRerouteUsage(), B=Multiply(A=Max(A=@n4, B=@n5), B=ScalarParameter() {parameterName=SpikesOpacities}))
  n6 = Add(
   A=MaterialFunctionCall() {materialFunction=&GetUserInterfaceUV}.Normalized UV
   B=AppendVector(
    A=ScalarParameter() {defaultValue=-0.184, group=Outside, parameterName=OffsetX_01, sortPriority=29}
    B=ScalarParameter() {defaultValue=0.04, group=Outside, parameterName=OffsetY_01}
   )
  )
  n7 = ScalarParameter() {defaultValue=1.8427, group=Outside, parameterName=Rotation_01}
  n8 = ScalarParameter() {defaultValue=0.58, group=Outside, parameterName=Perspective_01}
  n9 = MaterialFunctionCall!(
   UVs=Rotator(
    Coordinate=MaterialFunctionCall!(UV=@n6, Perspective Rotation=@n7, Perspective Amount=@n8) {materialFunction=/FortUILibrary/MaterialFunctions/Basics/MF_UI_Perspective}.Perspective UVs
    Time=ScalarParameter() {defaultValue=1.4248, group=Outside, parameterName=Rotator_01}
   ) {speed=3.1416}
   Radius=ScalarParameter() {defaultValue=1.2973, group=Outside, parameterName=Radius_01}
   Sides=Constant() {r=3}
   InnerAngle
  ) {materialFunction=/FortUILibrary/MaterialFunctions/Shapes/SDFs/MF_SDF_RegularPolygon}
  n10 = MaterialFunctionCall() {materialFunction=&GetUserInterfaceUV}
  n11 = AppendVector(
   A=ScalarParameter() {defaultValue=-0.4389, group=Inside, parameterName=OffsetX_02}
   B=ScalarParameter() {defaultValue=0.024, group=Inside, parameterName=OffsetY_02}
  )
  n12 = MaterialFunctionCall!(
   UVs=MaterialFunctionCall(
    UV=Add(A=@n10.Normalized UV, B=@n11)
    Perspective Rotation=ScalarParameter() {defaultValue=0.24, group=Inside, parameterName=Rotation_02}
    Perspective Amount=ScalarParameter() {defaultValue=-1.1688, group=Inside, parameterName=Pespective_02}
   ) {materialFunction=&MF_UI_Perspective}.Perspective UVs
   TopWidth=ScalarParameter() {defaultValue=0.3, group=Inside, parameterName=TopWidth_02}
   BottomWidth=ScalarParameter() {defaultValue=0.3, group=Inside, parameterName=BottomWidth_02}
   Height=ScalarParameter() {defaultValue=0.133, group=Inside, parameterName=Height_02}
   OutlineThickness
   OutlineAlignment
   GlowMin
   GlowMax
  ) {materialFunction=/FortUILibrary/MaterialFunctions/Shapes/MF_UI_IsoscelesTrapezoid}
  NamedRerouteDeclaration(None=MaterialFunctionCall(DistanceField=MaterialFunctionCall!(DistanceFieldA=@n9.SDF, DistanceFieldB=@n12.SDF, BlendAmount) {materialFunction=/FortUILibrary/MaterialFunctions/Shapes/Modifiers/MF_UI_SDF_Subtraction}.Exact) {materialFunction=&MF_UI_AntiAliasing}.Inner) {name=Bracket, nodeColor=<rgba 1 0.9375 0 1>}
  n13 = ScalarParameter() {parameterName=Spike_Offset}
  n14 = Add(A=Constant() {r=-0.12}, B=@n13)
  n15 = Add(A=Constant() {r=0.22}, B=@n13)
  n16 = MaterialFunctionCall(
   UVs=Add(A=MaterialFunctionCall() {materialFunction=&GetUserInterfaceUV}.Normalized UV, B=AppendVector(A=@n14, B=@n15)) {constB=0}
   Texture Scale=ScalarParameter() {defaultValue=0.12, parameterName=TopSpike_Scale}
  ) {materialFunction=&ScaleUVsByCenter}
  n17 = Add(A=@n16.UVs, B=AppendVector(A=Constant() {r=-0.088}, B=Constant() {r=-0.352}))
  n18 = Constant() {r=0.206}
  n19 = Constant() {r=0.316}
  n20 = MaterialFunctionCall(
   UVs=Rotator(
    Coordinate=MaterialFunctionCall(UV=@n17, Perspective Rotation=@n18, Perspective Amount=@n19) {materialFunction=&MF_UI_Perspective}.Perspective UVs
    Time=Constant() {r=0.1471}
   ) {speed=3.1416}
   Radius=Constant() {r=1.5967}
   Sides=Constant() {r=3}
  ) {materialFunction=&MF_SDF_RegularPolygon}
  n21 = Add(A=@n16.UVs, B=AppendVector(A=Constant() {r=-0.7438}, B=Constant() {r=-0.632}))
  n22 = Constant() {r=1.1934}
  n23 = Constant() {r=-0.764}
  n24 = MaterialFunctionCall(
   UVs=Rotator(
    Coordinate=MaterialFunctionCall(UV=@n21, Perspective Rotation=@n22, Perspective Amount=@n23) {materialFunction=&MF_UI_Perspective}.Perspective UVs
    Time=Constant() {r=1.6244}
   ) {speed=3.1416}
   Radius=Constant() {r=1.5154}
   Sides=Constant() {r=3}
  ) {materialFunction=&MF_SDF_RegularPolygon}
  NamedRerouteDeclaration(None=MaterialFunctionCall(DistanceField=MaterialFunctionCall(DistanceFieldA=@n20.SDF, DistanceFieldB=@n24.SDF) {materialFunction=&MF_UI_SDF_Subtraction}.Exact) {materialFunction=&MF_UI_AntiAliasing}.Inner) {name=TopSpike, nodeColor=<rgba 0.6172 0 1 1>}
  n25 = ScalarParameter() {parameterName=Spike_Offset}
  n26 = Add(A=Constant() {r=0.08}, B=@n25)
  n27 = Subtract(A=Constant() {r=-0.255}, B=@n25)
  n28 = MaterialFunctionCall(
   UVs=Add(A=MaterialFunctionCall() {materialFunction=&GetUserInterfaceUV}.Normalized UV, B=AppendVector(A=@n26, B=@n27)) {constB=0}
   Texture Scale=ScalarParameter() {defaultValue=0.21, parameterName=BottomSpike_Scale}
  ) {materialFunction=&ScaleUVsByCenter}
  n29 = Add(A=@n28.UVs, B=AppendVector(A=Constant() {r=-0.096}, B=Constant() {r=0.36}))
  n30 = Constant() {r=0.806}
  n31 = Constant() {r=0.692}
  n32 = MaterialFunctionCall(
   UVs=Rotator(
    Coordinate=MaterialFunctionCall(UV=@n29, Perspective Rotation=@n30, Perspective Amount=@n31) {materialFunction=&MF_UI_Perspective}.Perspective UVs
    Time=Constant() {r=1.469}
   ) {speed=3.1416}
   Radius=Constant() {r=1.8058}
   Sides=Constant() {r=3}
  ) {materialFunction=&MF_SDF_RegularPolygon}
  n33 = Add(A=@n28.UVs, B=AppendVector(A=Constant() {r=-0.292}, B=Constant() {r=0.352}))
  n34 = Constant() {r=1.1942}
  n35 = Constant() {r=0.356}
  n36 = MaterialFunctionCall(
   UVs=Rotator(
    Coordinate=MaterialFunctionCall(UV=@n33, Perspective Rotation=@n34, Perspective Amount=@n35) {materialFunction=&MF_UI_Perspective}.Perspective UVs
    Time=Constant() {r=1.4148}
   ) {speed=3.1416}
   Radius=Constant() {r=1.6032}
   Sides=Constant() {r=3}
  ) {materialFunction=&MF_SDF_RegularPolygon}
  NamedRerouteDeclaration(None=MaterialFunctionCall(DistanceField=MaterialFunctionCall(DistanceFieldA=@n32.SDF, DistanceFieldB=@n36.SDF) {materialFunction=&MF_UI_SDF_Subtraction}.Exact) {materialFunction=&MF_UI_AntiAliasing}.Inner) {name=BottomSpike, nodeColor=<rgba 1 0.3047 0 1>}
 ]
}
