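# UI-domain translucent material for font rendering, named for the boss health bar text in the Athena HUD.
# A LinearGradient function's vertical gradient blends a light grey to a dark grey to shade the glyphs.
# The same gradient, scaled and inverted, multiplies a FontSignedDistance opacity, so the signed-distance text also partially fades along the vertical axis.
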Material /Game/Athena/HUD/Actors/Materials/M_UI_BossHealthBar_Font
graph {
 properties {bAutomaticallySetUsageInEditor=false, blendMode=BLEND_Translucent, materialDomain=MD_UI, usedShadingModels=Default Lit}
 nodes [
  n1 = MaterialFunctionCall() {materialFunction=&LinearGradient}
  LinearInterpolate(
   A=Constant3Vector() {constant=<rgba 0.875 0.875 0.875 1>}
   B=Constant3Vector() {constant=<rgba 0.0938 0.0938 0.0938 1>}
   Alpha=Multiply(A=@n1.VGradient) {constB=0.9}
  )
  n2 = Power(Base=@n1.VGradient) {constExponent=1}
  Multiply(A=OneMinus(None=Multiply(A=@n2) {constB=0.8}), B=FontSignedDistance().Implicit Opacity)
 ]
}
