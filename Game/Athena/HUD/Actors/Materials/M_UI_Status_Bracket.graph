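# Translucent UI material for a HUD status bracket, drawn procedurally from signed distance fields.
# A three-sided polygon and an isosceles trapezoid each get their own offset and perspective warp, with the polygon also rotated.
# The trapezoid is subtracted from the polygon, and the anti-aliased result is scaled by OpacityMultiplier.
# An Attacked parameter blends the color from an IsLocked-driven base toward the yellow-default Color parameter.
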Material /Game/Athena/HUD/Actors/Materials/M_UI_Status_Bracket
graph {
 properties {
  bAutomaticallySetUsageInEditor=false
  blendMode=BLEND_Translucent
  materialDomain=MD_UI
  previewMesh=/Engine/Transient.MaterialEditorMeshComponent_2
  usedShadingModels=Default Lit
 }
 nodes [
  LinearInterpolate(
   A=LinearInterpolate(A=Constant() {r=1}, Alpha=ScalarParameter() {parameterName=IsLocked})
   B=VectorParameter() {parameterName=Color, defaultValue=<rgba 0.9301 1 0.0103 1>}.RGB
   Alpha=ScalarParameter() {parameterName=Attacked}
  )
  n1 = Add(
   A=MaterialFunctionCall() {materialFunction=&GetUserInterfaceUV}.Normalized UV
   B=AppendVector(
    A=ScalarParameter() {defaultValue=0.008, group=Outside, parameterName=OffsetX_01, sortPriority=29}
    B=ScalarParameter() {defaultValue=0.048, group=Outside, parameterName=OffsetY_01}
   )
  )
  n2 = ScalarParameter() {defaultValue=0.84, group=Outside, parameterName=Rotation_01}
  n3 = ScalarParameter() {defaultValue=0.356, group=Outside, parameterName=Perspective_01}
  n4 = MaterialFunctionCall!(
   UVs=Rotator(
    Coordinate=MaterialFunctionCall!(UV=@n1, Perspective Rotation=@n2, Perspective Amount=@n3) {materialFunction=/FortUILibrary/MaterialFunctions/Basics/MF_UI_Perspective}.Perspective UVs
    Time=ScalarParameter() {defaultValue=1.4156, group=Outside, parameterName=Rotator_01}
   ) {speed=3.1416}
   Radius=ScalarParameter() {defaultValue=1.7138, group=Outside, parameterName=Radius_01}
   Sides=Constant() {r=3}
   InnerAngle
  ) {materialFunction=/FortUILibrary/MaterialFunctions/Shapes/SDFs/MF_SDF_RegularPolygon}
  n5 = MaterialFunctionCall() {materialFunction=&GetUserInterfaceUV}
  n6 = AppendVector(
   A=ScalarParameter() {defaultValue=-0.184, group=Inside, parameterName=OffsetX_02}
   B=ScalarParameter() {defaultValue=0.088, group=Inside, parameterName=OffsetY_02}
  )
  n7 = MaterialFunctionCall!(
   UVs=MaterialFunctionCall(
    UV=Add(A=@n5.Normalized UV, B=@n6)
    Perspective Rotation=ScalarParameter() {defaultValue=0.24, group=Inside, parameterName=Rotation_02}
    Perspective Amount=ScalarParameter() {defaultValue=-1.1688, group=Inside, parameterName=Pespective_02}
   ) {materialFunction=&MF_UI_Perspective}.Perspective UVs
   TopWidth=ScalarParameter() {defaultValue=0.3, group=Inside, parameterName=TopWidth_02}
   BottomWidth=ScalarParameter() {defaultValue=0.3, group=Inside, parameterName=BottomWidth_02}
   Height=ScalarParameter() {defaultValue=0.25, group=Inside, parameterName=Height_02}
   OutlineThickness
   OutlineAlignment
   GlowMin
   GlowMax
  ) {materialFunction=/FortUILibrary/MaterialFunctions/Shapes/MF_UI_IsoscelesTrapezoid}
  Multiply(
   A=MaterialFunctionCall(DistanceField=MaterialFunctionCall!(DistanceFieldA=@n4.SDF, DistanceFieldB=@n7.SDF, BlendAmount) {materialFunction=/FortUILibrary/MaterialFunctions/Shapes/Modifiers/MF_UI_SDF_Subtraction}.Exact) {materialFunction=&MF_UI_AntiAliasing}.Inner
   B=ScalarParameter() {defaultValue=0.6, parameterName=OpacityMultiplier}
  ) {constB=0.6}
 ]
}
