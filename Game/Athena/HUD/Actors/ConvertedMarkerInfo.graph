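# HUD marker widget showing either a player's social avatar with a pin-colored squad diamond or a tinted NPC command icon.
# Init events switch between the two views, async-loading the avatar into a dynamic material or the icon chosen by ping command type.
# Commands without icons collapse the command image.
# SetBackgroundVisibility shows or collapses the whole widget; unless overridden, talking or an active DBNO pulse timer also keeps it visible.
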WidgetBlueprint /Game/Athena/HUD/Actors/ConvertedMarkerInfo
graphs [
 EventGraph {nodes [
   n1 = InitNPCCommandIcon!() -> (OutputDelegate, then, Color:Linear Color Structure) {is=title K2Node_CustomEvent}
   > UpdateCommandIcon!(execute, self, PingCommand:PingCommandType Enum=BackToDefault, Color:Linear Color Structure=@n1.Color) -> (then)
   > SetActiveWidgetIndex(self:Common Visibility Switcher Object Reference=Get ImageSwitcher() -> (ImageSwitcher:Common Visibility Switcher Object Reference), Index=1)
   > SetVisibility(self=Get Image_Command() -> (Image_Command:Image Object Reference), InVisibility=Collapsed)
   n2 = InitPlayerIconState!() -> (OutputDelegate, then, PSA:Fort Player State Athena Object Reference) {is=title K2Node_CustomEvent}
   > SetColorAndOpacity(
    self:Image Object Reference=Get SquadMarker() -> (SquadMarker:Image Object Reference)
    InColorAndOpacity:Linear Color Structure=GetPinColor!(self:Fort Player State Zone Object Reference=@n2.PSA) -> (ReturnValue:Linear Color Structure)
   )
   > Cast To FortPlayerControllerAthena(Object=GetOwningPlayer())
   > n3 = Async Load Asset(Asset=GetSocialAvatarBrush!(self:Fort Player State Zone Object Reference=@n2.PSA, bSmallImage:bool=true) -> (ReturnValue:Texture 2D Soft Object Reference))
   n4 = Cast To Texture2D(execute=[@n3.Completed, @n3.then], Object=@n3)
   > n5 = GetDynamicMaterial(self=Get Image_Avatar() -> (Image_Avatar:Image Object Reference))
   > SetTextureParameterValue(self=@n5, ParameterName=SmallPreview, Value=@n4)
   > SetActiveWidgetIndex(self=Get ImageSwitcher(), Index=0)
   n6 = OverrideShowBackground!() -> (OutputDelegate, then, bShowBackground:bool=~) {is=title K2Node_CustomEvent}
   > SetBackgroundVisibility!(execute, self, bOverride:bool=true, bBackgroundVisible:bool=@n6.bShowBackground) -> (then)
   n7 = UpdateCommandIcon!() -> (OutputDelegate, then, PingCommand:PingCommandType Enum=GoTo, Color:Linear Color Structure) {is=title K2Node_CustomEvent}
   > n8 = Switch on PingCommandType!(execute, Selection:PingCommandType Enum=@n7.PingCommand) -> (GoTo:Exec, BackToMe:Exec, BackToDefault:Exec, HoldPosition:Exec, Revive:Exec, Dismiss:Exec, Item:Exec, Attack:Exec, Invalid:Exec) {is=title K2Node_SwitchEnum}
   n9 = Async Load Asset(
    execute=[@n8.Attack, @n8.BackToMe, @n8.GoTo, @n8.HoldPosition, @n8.Revive]
    Asset=Select!(
     GoTo:Texture 2D Soft Object Reference=Get Icon_Move() -> (Icon_Move:Texture 2D Soft Object Reference)
     BackToMe:Texture 2D Soft Object Reference=Get Icon_BackToMe() -> (Icon_BackToMe:Texture 2D Soft Object Reference)
     BackToDefault:Texture 2D Soft Object Reference
     HoldPosition:Texture 2D Soft Object Reference=Get Icon_Hold() -> (Icon_Hold:Texture 2D Soft Object Reference)
     Revive:Texture 2D Soft Object Reference=Get Icon_Revive() -> (Icon_Revive:Texture 2D Soft Object Reference)
     Dismiss:Texture 2D Soft Object Reference
     Item:Texture 2D Soft Object Reference
     Attack:Texture 2D Soft Object Reference=Get Icon_Attack() -> (Icon_Attack:Texture 2D Soft Object Reference)
     Invalid:Texture 2D Soft Object Reference
     Index:PingCommandType Enum=@n7.PingCommand
    ) -> (ReturnValue:Texture 2D Soft Object Reference)
   )
   n10 = Cast To Texture2D(execute=[@n9.Completed, @n9.then], Object=@n9)
   n11 = Get Image_Command()
   > SetBrushFromTexture!(execute, self:Image Object Reference=@n11, Texture:Texture 2D Object Reference=@n10, bMatchSize:bool=false) -> (then)
   > SetColorAndOpacity(self=@n11, InColorAndOpacity=@n7.Color)
   > SetActiveWidgetIndex(self=Get ImageSwitcher(), Index=1)
   > SetVisibility(self=Get Image_Command(), InVisibility=HitTestInvisible)
   SetVisibility(execute=[@n8.BackToDefault, @n8.Dismiss, @n8.Invalid, @n8.Item], self=Get Image_Command(), InVisibility=Collapsed)
   Event PreConstruct()
  ]}
 SetBackgroundVisibility {nodes [
   n1 = SetBackgroundVisibility!() -> (then, bOverride:bool=~, bBackgroundVisible:bool=~) {is=title K2Node_FunctionEntry}
   > n2 = Set bShowBackgroundOverridden(bShowBackgroundOverridden:bool=@n1) -> (Output_Get:bool)
   > n3 = Sequence()
   > n4 = Branch(Condition=@n2)
   > Set bLocal_BackgroundVisible(bLocal_BackgroundVisible:bool=Get bBackgroundVisible() -> (bBackgroundVisible:bool)) -> (Output_Get:bool)
   n5 = Get DBNOPulseTimer() -> (DBNOPulseTimer:Timer Handle Structure)
   Set bLocal_BackgroundVisible(
    execute=@n4.else
    bLocal_BackgroundVisible=OR Boolean!(
     A=Get bBackgroundVisible()
     B=Get bIsTalking() -> (bIsTalking:bool)
     C:bool=Is Valid Timer Handle!(Handle:Timer Handle Structure=@n5) -> (ReturnValue:bool)
    ) -> (ReturnValue)
   )
   SetVisibility(
    execute=@n3.then_1
    InVisibility=Select!(
     Option 0:ESlateVisibility Enum=Collapsed
     Option 1:ESlateVisibility Enum=SelfHitTestInvisible
     Index:bool=Get bLocal_BackgroundVisible() -> (bLocal_BackgroundVisible:bool)
    ) -> (ReturnValue:ESlateVisibility Enum)
   )
   Get WeakPSA() -> (WeakPSA:Fort Player State Athena Weak object)
  ]}
]
variables [
 bEnemyVersion {type=boolean}
 bIsTalking {type=boolean}
 bShowBackgroundOverridden {type=boolean}
 DBNOPulseTimer {type=TimerHandle}
 Icon_Attack {type=Texture2D, default=/Game/Athena/HUD/Art/squadArt/T_UI_Squad_Attack}
 Icon_BackToMe {type=Texture2D, default=/Game/Athena/HUD/Art/squadArt/T_UI_NPCCom_BacktoMe}
 Icon_Hold {type=Texture2D, default=/Game/Athena/HUD/Art/squadArt/T_UI_NPCCom_Hold_Position}
 Icon_Move {type=Texture2D, default=/Game/Athena/HUD/Art/squadArt/T_UI_NPCCom_Move}
 Icon_Revive {type=Texture2D, default=/Game/Athena/HUD/Art/squadArt/T_UI_NPCCom_Revive}
 RelevantPlaylistTags {type=GameplayTagContainer, default {gameplayTags [{tagName=Athena.Playlist.Papaya}]}}
]
widgets {
 info {inheritedWidgetCount=0, parentClass=&CommonUserWidget, rootWidgetClass=&Overlay, widgetCount=6}
 widgets [
  Overlay_0 {widgetClassPath=&Overlay, widget=Overlay_0}
  ImageSwitcher {
   widgetClassPath=&CommonVisibilitySwitcher
   bIsVariable=true
   parent=Overlay_0
   widget=ImageSwitcher
   slot {class=OverlaySlot, properties {horizontalAlignment=HAlign_Center, verticalAlignment=VAlign_Center}}
  }
  HorizontalBox_0 {widgetClassPath=&HorizontalBox, parent=ImageSwitcher, widget=HorizontalBox_0, slot {class=CommonVisibilitySwitcherSlot}}
  Image_Avatar {
   widgetClassPath=&Image
   bIsVariable=true
   parent=HorizontalBox_0
   widget=Image_Avatar
   properties {
    visibility=HitTestInvisible
    brush {
     bIsDynamicallyLoaded=false
     drawAs=Image
     imageType=NoImage
     mirroring=NoMirror
     resourceName=None
     resourceObject=/Game/UI/SocialPanel/Art/Materials/UI_UserPortrait_Buffet
     tiling=NoTile
     imageSize=<48 48>
     margin=<ltrb 0 0 0 0>
     outlineSettings {
      bUseBrushTransparency=false
      roundingType=HalfHeightRadius
      width=0
      color {colorUseRule=UseColor_Specified, specifiedColor=<rgba 0 0 0 0>}
      cornerRadii=<0 0 0 0>
     }
     tintColor {colorUseRule=UseColor_Specified, specifiedColor=<rgba 1 1 1 1>}
     uVRegion {bIsValid=false, max=<0 0>, min=<0 0>}
    }
   }
   slot {class=HorizontalBoxSlot, properties {horizontalAlignment=HAlign_Left, verticalAlignment=VAlign_Top}}
  }
  SquadMarker {
   widgetClassPath=&Image
   bIsVariable=true
   parent=HorizontalBox_0
   widget=SquadMarker
   properties {
    visibility=SelfHitTestInvisible
    brush {
     bIsDynamicallyLoaded=false
     drawAs=Image
     imageType=NoImage
     mirroring=NoMirror
     resourceName=None
     resourceObject=/Game/Athena/HUD/Art/T_UI_Indicator_Diamond
     tiling=NoTile
     imageSize=<14 16>
     margin=<ltrb 0 0 0 0>
     outlineSettings {
      bUseBrushTransparency=false
      roundingType=HalfHeightRadius
      width=0
      color {colorUseRule=UseColor_Specified, specifiedColor=<rgba 0 0 0 0>}
      cornerRadii=<0 0 0 0>
     }
     tintColor {colorUseRule=UseColor_Specified, specifiedColor=<rgba 1 1 1 1>}
     uVRegion {bIsValid=false, max=<0 0>, min=<0 0>}
    }
   }
   slot {class=HorizontalBoxSlot, properties {horizontalAlignment=HAlign_Center, verticalAlignment=VAlign_Center}}
  }
  Image_Command {
   widgetClassPath=&Image
   bIsVariable=true
   parent=ImageSwitcher
   widget=Image_Command
   properties {
    visibility=HitTestInvisible
    brush {
     bIsDynamicallyLoaded=false
     drawAs=Image
     imageType=NoImage
     mirroring=NoMirror
     resourceName=None
     resourceObject=&T_UI_NPCCom_Move
     tiling=NoTile
     imageSize=<48 48>
     margin=<ltrb 0 0 0 0>
     outlineSettings {
      bUseBrushTransparency=false
      roundingType=HalfHeightRadius
      width=0
      color {colorUseRule=UseColor_Specified, specifiedColor=<rgba 0 0 0 0>}
      cornerRadii=<0 0 0 0>
     }
     tintColor {colorUseRule=UseColor_Specified, specifiedColor=<rgba 1 1 1 1>}
     uVRegion {bIsValid=false, max=<0 0>, min=<0 0>}
    }
   }
   slot {class=CommonVisibilitySwitcherSlot, properties {horizontalAlignment=HAlign_Right, verticalAlignment=VAlign_Bottom}}
  }
 ]
}
