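# World marker indicator widget for converted NPCs, showing an info icon above an arrow to the locally controlled player who converted them.
# When marker data is set, it reads the marked actor's convert component and collapses the icon if validity, local-control or class checks fail.
# Wildlife gets the converter's player icon state.
# Tandem NPCs get a command icon in the component's UI color, with ping-command events updating icon and arrow visibility.
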WidgetBlueprint /Game/Athena/HUD/Actors/NPCConvertedIndicatorMarker
graphs [
 EventGraph {nodes [
   n1 = OnPingCommandEvent!() -> (OutputDelegate, then, CommandType:PingCommandType Enum=GoTo) {is=title K2Node_CustomEvent}
   > UpdateCommandIcon!(
    execute
    self:Converted Marker Info Object Reference=Get RidingMarkerInfo() -> (RidingMarkerInfo:Converted Marker Info Object Reference)
    PingCommand:PingCommandType Enum=@n1.CommandType
    Color:Linear Color Structure=Get CachedColor() -> (CachedColor:Linear Color Structure)
   ) -> (then)
   > SetVisibility(
    self=Get IrwinIndicator() -> (IrwinIndicator:Image Object Reference)
    InVisibility=Select!(
     GoTo:ESlateVisibility Enum=Collapsed
     BackToMe:ESlateVisibility Enum=Collapsed
     BackToDefault:ESlateVisibility Enum=HitTestInvisible
     HoldPosition:ESlateVisibility Enum=Collapsed
     Revive:ESlateVisibility Enum=Collapsed
     Dismiss:ESlateVisibility Enum=HitTestInvisible
     Item:ESlateVisibility Enum=HitTestInvisible
     Attack:ESlateVisibility Enum=Collapsed
     Invalid:ESlateVisibility Enum=HitTestInvisible
     Index:PingCommandType Enum=@n1.CommandType
    ) -> (ReturnValue)
   )
   n2 = Event OnSetMarkerData!() -> (OutputDelegate, then, FortWorldMarkerData:"Fort World Marker Data Structure (by ref)") {is=title K2Node_Event}
   > Parent: OnSetMarkerData!(execute, FortWorldMarkerData:"Fort World Marker Data Structure (by ref)"=@n2.FortWorldMarkerData) -> (then) {is=title K2Node_CallParentFunction}
   n3 = Break Fort World Marker Data!(FortWorldMarkerData:"Fort World Marker Data Structure (by ref)"=@n2.FortWorldMarkerData) -> (
    MarkerID:Marker ID Structure
    Owner:Fort Player State Weak object
    MarkerType:EFortWorldMarkerType Enum
    BasePositionOffset:Vector=~
    WorldNormal:Vector=~
    ItemDefinition:Item Definition Base Object Reference
    ItemCount:int
    PinnedQuestIndex:int
    MarkedActor:Actor Weak object
    MarkedActorClass:Actor Soft Class Reference
    MarkerTag:Gameplay Tag Structure
    bUseHoveredMarkerDetail:bool=~
   ) {is=title K2Node_BreakStruct}
   n4 = GetComponentByClass!(self:Actor Object Reference=@n3.MarkedActor, ComponentClass:Actor Component Class Reference=/Script/FortniteGame.FortPawnComponent_Convert) -> (ReturnValue:Fort Pawn Component Convert Object Reference)
   > n5 = Is Valid(InputObject=@n4)
   n6 = GetConverterPawn!(self:Fort Pawn Component Convert Object Reference=@n4) -> (ReturnValue:Fort Player Pawn Object Reference)
   > n7 = Is Valid(InputObject=@n6)
   > n8 = Branch(Condition=IsLocallyControlled!(self:Pawn Object Reference=@n6) -> (ReturnValue:bool))
   > n9 = AttemptGetPlayerStateForPawn!(execute, FortPawn:Fort Player Pawn Object Reference=@n6) -> (then, ReturnValue:Fort Player State Object Reference)
   > n10 = Is Valid(InputObject=@n9)
   > n11 = Cast To FortPlayerStateAthena!(execute, Object=@n9) -> (then, CastFailed, AsFort Player State Athena:Fort Player State Athena Object Reference) {is=title K2Node_DynamicCast}
   > n12 = Branch(Condition=IsAWildlifePawn!(Actor:Actor Object Reference=@n3.MarkedActor) -> (bWildlife:bool))
   > n13 = InitPlayerIconState!(execute, self:Converted Marker Info Object Reference=Get RidingMarkerInfo(), PSA:Fort Player State Athena Object Reference=@n11) -> (then)
   SetVisibility(execute=[@n16.then, @n13.then], self=Get RidingMarkerInfo(), InVisibility=HitTestInvisible)
   n14 = "IsA ( soft )"!(
    execute=@n12.else
    Object:Object Reference=@n3.MarkedActor
    SoftClass:Object Soft Class Reference=/Game/Athena/AI/NPCs/Tandem/Base/Pawns/BP_PlayerPawn_Tandem.BP_PlayerPawn_Tandem_C
   ) -> (True:Exec, False:Exec)
   > Bind Event to On Ping Command Event(self:Fort Pawn Component Convert Object Reference=@n4, Delegate=@n1) {is=title K2Node_AddDelegate}
   > n15 = Set CachedColor(CachedColor:Linear Color Structure=GetUIColor!(self:Fort Pawn Component Convert Object Reference=@n4) -> (ReturnValue:Linear Color Structure)) -> (Output_Get:Linear Color Structure)
   > n16 = InitNPCCommandIcon!(execute, self:Converted Marker Info Object Reference=Get RidingMarkerInfo(), Color:Linear Color Structure=@n15) -> (then)
   SetVisibility(
    execute=[@n14.False, @n11.CastFailed, @n10.Is Not Valid, @n8.else, @n7.Is Not Valid, @n5.Is Not Valid]
    self=Get RidingMarkerInfo()
    InVisibility=Collapsed
   )
   Event PreConstruct()
   Event Construct()
   Event Tick()
  ]}
 IsAWildlifePawn {nodes [
   n1 = IsAWildlifePawn!() -> (then, Actor:Actor Object Reference) {is=title K2Node_FunctionEntry}
   > Return Node!(
    execute
    bWildlife:bool=IsA!(Object:Object Reference=@n1, ObjectClass:Object Class Reference=/Game/Athena/AI/Pawns/NPC/Parents/NPC_Pawn_Parent.NPC_Pawn_Parent_C) -> (ReturnValue:bool)
   )
  ]}
]
variables [CachedColor {type=LinearColor, default=<rgba 0 0 0 1>}]
widgets {
 info {inheritedWidgetCount=5, parentClass=/Script/FortniteUI.AthenaMarkedActorIndicator, rootWidgetClass=&VerticalBox, widgetCount=8}
 widgets [
  VerticalBox_0 {widgetClassPath=&VerticalBox, widget=VerticalBox_0}
  RidingMarkerInfo {
   widgetClassPath=&/ConvertedMarkerInfo
   bIsVariable=true
   parent=VerticalBox_0
   widget=RidingMarkerInfo
   slot {class=VerticalBoxSlot, properties {horizontalAlignment=HAlign_Center, padding=<ltrb 0 0 0 -30>}}
  }
  IrwinIndicator {
   widgetClassPath=&Image
   bIsVariable=true
   parent=VerticalBox_0
   widget=IrwinIndicator
   properties {brush {
     bIsDynamicallyLoaded=false
     drawAs=Image
     imageType=FullColor
     mirroring=NoMirror
     resourceName=None
     resourceObject=/Game/Athena/HUD/Team/MeshIndicators/T_TeamIndicatorArrow
     tiling=NoTile
     imageSize=<64 64>
     margin=<ltrb 0 0 0 0>
     outlineSettings {
      bUseBrushTransparency=false
      roundingType=HalfHeightRadius
      width=0
      color {colorUseRule=UseColor_Specified, specifiedColor=<rgba 0 0 0 0>}
      cornerRadii=<0 0 0 0>
     }
     tintColor {colorUseRule=UseColor_Specified, specifiedColor=<rgba 1 1 1 1>}
     uVRegion {bIsValid=false, max=<0 0>, min=<0 0>}
    }}
   slot {class=VerticalBoxSlot}
  }
  Switcher_IndicatorSize {widgetClassPath=&WidgetSwitcher, bInherited=true, widget=null}
  Panel_SmallIndicator {widgetClassPath=/Script/UMG.PanelWidget, bInherited=true, widget=null}
  Panel_LargeIndicator {widgetClassPath=&PanelWidget, bInherited=true, widget=null}
  Text_Distance {widgetClassPath=&CommonTextBlock, bInherited=true, widget=null}
  TutorialHighlight {widgetClassPath=&TutorialHighlight}
 ]
}
