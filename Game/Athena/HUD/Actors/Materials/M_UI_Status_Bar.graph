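# Translucent UI material for a HUD status bar: a trapezoid with a two-color gradient progress fill, outline, background and optional segment dividers.
# An SDF trapezoid provides anti-aliased bar and outline masks, while padded, aspect-corrected UVs compared against Progress give the fill mask.
# A rotated gradient texture, stretched by segment count and smoothstepped, forms dividers that either cut opacity or take SegmentColor.
# IsUnlocked lightens part of the unfilled bar toward white.
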Material /Game/Athena/HUD/Actors/Materials/M_UI_Status_Bar
graph {
 properties {bAutomaticallySetUsageInEditor=false, blendMode=BLEND_Translucent, materialDomain=MD_UI, usedShadingModels=Default Lit}
 nodes [
  n1 = Add(
   A=MaterialFunctionCall(UVs=TextureCoordinate(), "Rotation Angle (0-1)"=ScalarParameter() {defaultValue=0.25, parameterName=SegmentAngle}) {materialFunction=&CustomRotator}.Rotated Values
   B=AppendVector(A=Constant(), B=ScalarParameter() {defaultValue=0.004, parameterName=SegmentsOffset})
  ) {constB=0}
  n2 = ScalarParameter() {parameterName=Segments}
  n3 = AppendVector(A=Constant() {desc=stretch vertical to eliminate artifacts, r=0.01}, B=@n2)
  n4 = Multiply(A=@n2, B=ScalarParameter() {defaultValue=0.0007, parameterName=SegmentEdge})
  n5 = ScalarParameter() {defaultValue=0, parameterName=SegmentEdgeSoftness}
  n6 = MaterialFunctionCall(
   Alpha=TextureSample(UVs=Multiply(A=@n1, B=@n3)) {texture=/Game/Global/Textures/Default/Masks_and_Gradients/T_UI_ReticleSegmentGradient}.R
   Min=@n4
   Max=Add(A=@n4, B=Multiply(A=@n2, B=@n5))
  ) {materialFunction=&SmoothStep}
  n7 = NamedRerouteUsage()
  LinearInterpolate(
   A=If(A=@n2, B=Constant() {r=1}, "A > B"=Multiply(A=@n6.Result, B=@n7), "A == B"=@n7, "A < B"=@n7)
   B=NamedRerouteUsage()
   Alpha=ScalarParameter() {parameterName=EnableColoredSegments}
  )
  n8 = VectorParameter() {parameterName=Color_Bg, defaultValue=<rgba 0.0024 0.0018 0.006 1>}
  n9 = LinearInterpolate(A=@n8.RGB, B=Constant() {r=1}, Alpha=ScalarParameter() {parameterName=IsUnlocked})
  n10 = NamedRerouteUsage()
  n11 = VectorParameter() {parameterName=ColorA, defaultValue=<rgba 0.9301 0.0865 0.0185 1>}
  n12 = VectorParameter() {parameterName=ColorB, defaultValue=<rgba 0.9301 0.2051 0.0168 1>}
  n13 = MaterialFunctionCall() {materialFunction=&LinearGradient}
  LinearInterpolate(
   A=LinearInterpolate(
    A=LinearInterpolate(A=@n9, B=@n8.RGB, Alpha=@n10)
    B=LinearInterpolate(A=@n11.RGB, B=@n12.RGB, Alpha=@n13.UGradient)
    Alpha=NamedRerouteUsage()
   )
   B=VectorParameter() {parameterName=SegmentColor, defaultValue=<rgba 0 0 0 1>}.RGB
   Alpha=Multiply(A=NamedRerouteUsage(), B=ScalarParameter() {parameterName=EnableColoredSegments})
  ) {constB=0}
  NamedRerouteDeclaration(None=OneMinus(None=@n6.Result)) {name=Segments, nodeColor=<rgba 0 0 0 1>}
  n14 = MaterialFunctionCall!(
   UVs=MaterialFunctionCall!(UVs=MaterialFunctionCall() {materialFunction=&GetUserInterfaceUV}.Normalized UV) {materialFunction=/FortUILibrary/MaterialFunctions/Basics/MF_UI_9SliceUVs}.UVs
   TopWidth=ScalarParameter() {defaultValue=0.6, parameterName=TrapezoidTopWidth}
   BottomWidth=ScalarParameter() {defaultValue=0.99, parameterName=TrapezoidBottomWidth}
   Height=ScalarParameter() {defaultValue=0.99, parameterName=TrapezoidHeight}
   OutlineThickness
   OutlineAlignment
   GlowMin
   GlowMax
  ) {materialFunction=/FortUILibrary/MaterialFunctions/Shapes/MF_UI_IsoscelesTrapezoid}
  NamedRerouteDeclaration(None=MaterialFunctionCall(DistanceField=@n14.SDF) {materialFunction=&MF_UI_AntiAliasing}.Inner) {name=BarMask, nodeColor=<rgba 0 0 0 1>}
  n15 = ScalarParameter() {defaultValue=0.01, parameterName=BarPadding}
  n16 = MaterialFunctionCall() {materialFunction=&GetUserInterfaceUV}
  n17 = Multiply(A=@n15, B=Divide(A=ComponentMask(None=@n16.Pixel Size) {r=true}, B=ComponentMask(None=@n16.Pixel Size) {g=true}))
  n18 = MaterialFunctionCall(
   UVs=MaterialFunctionCall() {materialFunction=&GetUserInterfaceUV}.Normalized UV
   Texture Scale=OneMinus(None=AppendVector(A=@n15, B=@n17))
  ) {materialFunction=&ScaleUVsByCenter}
  n19 = Subtract(
   A=ComponentMask(None=NamedRerouteDeclaration(None=@n18.UVs) {name=BarUVs, nodeColor=<rgba 0 0 0 1>}) {r=true}
   B=ScalarParameter() {defaultValue=0.5, group=Fill, parameterName=Progress, sortPriority=0}
  ) {constB=0.75}
  NamedRerouteDeclaration(None=MaterialFunctionCall(DistanceField=NamedRerouteDeclaration(None=@n19) {name=ProgressFillSDF, nodeColor=<rgba 0 0 0 1>}) {materialFunction=&MF_UI_AntiAliasing}.Inner) {name=ProgressFillMask, nodeColor=<rgba 0 0 0 1>}
  NamedRerouteDeclaration(None=MaterialFunctionCall(DistanceField=@n14.SDF, Threshold=Constant() {r=-0.05}) {materialFunction=&MF_UI_AntiAliasing}.Inner) {name=BarOutline, nodeColor=<rgba 0 0 0 1>}
  n20 = LinearInterpolate(
   A=ScalarParameter() {defaultValue=0.2, parameterName=OutlineOpacity}
   B=ScalarParameter() {defaultValue=0.25, parameterName=BackgroundOpacity}
   Alpha=NamedRerouteUsage()
  )
  n21 = NamedRerouteUsage()
  NamedRerouteDeclaration(None=LinearInterpolate(A=Multiply(A=@n20, B=@n21) {constA=0.5}, B=@n21, Alpha=NamedRerouteUsage()) {constA=0.5}) {name=Opacity, nodeColor=<rgba 0 0 0 1>}
 ]
}
