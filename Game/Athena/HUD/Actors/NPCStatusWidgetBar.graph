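# NPC status bar widget that stacks a main fill over a delayed delta fill, both driven through dynamic material parameters.
# SetPercent moves the main fill at once and starts a looping timer, evidently for BarAnimInterpolate, which steps the delta down to the target.
# SetPercentInstant sets both layers, SetNumPips sets segments, and PreConstruct applies colours.
# A quartile health-colour function exists, but its only call follows an unconnected Branch.
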WidgetBlueprint /Game/Athena/HUD/Actors/NPCStatusWidgetBar
graphs [
 BarAnimInterpolate {nodes [
   BarAnimInterpolate() {is=title K2Node_FunctionEntry}
   > n1 = Set PreviousProgressValue(PreviousProgressValue:double=float - float(A=Get PreviousProgressValue() -> (PreviousProgressValue:double), B=Get InterpCadence() -> (InterpCadence:double))) -> (Output_Get:double)
   > n2 = Branch(Condition="float <= float"(A=@n1, B=Get CurrentProgressValue() -> (CurrentProgressValue:double)) -> (ReturnValue=~))
   > Clear and Invalidate Timer by Handle(Handle=Get Timer_DeltaAnimateProgressUpdate() -> (Timer_DeltaAnimateProgressUpdate:Timer Handle Structure))
   > n3 = GetDynamicMaterial(self=Get ProgressBar_Delta() -> (ProgressBar_Delta:Image Object Reference))
   > SetScalarParameterValue(self=@n3, ParameterName=Progress, Value=Get CurrentProgressValue())
   > Set PreviousProgressValue(PreviousProgressValue=Get CurrentProgressValue())
   n4 = GetDynamicMaterial(execute=@n2.else, self=Get ProgressBar_Delta())
   > SetScalarParameterValue(self=@n4, ParameterName=Progress, Value=Get PreviousProgressValue())
  ]}
 EventGraph {nodes [
   n1 = SetPercentInstant!() -> (OutputDelegate, then, Value:double=~) {is=title K2Node_CustomEvent}
   > n2 = GetDynamicMaterial(self=Get ProgressBar() -> (ProgressBar:Image Object Reference))
   > SetScalarParameterValue(self=@n2, ParameterName=Progress, Value=@n1.Value)
   > n3 = GetDynamicMaterial(self=Get ProgressBar_Delta())
   > SetScalarParameterValue(self=@n3, ParameterName=Progress, Value=@n1.Value)
   > n4 = Set PreviousProgressValue(PreviousProgressValue=@n1.Value)
   > Set CurrentProgressValue(CurrentProgressValue:double=@n4) -> (Output_Get:double)
   n5 = SetNumPips!() -> (OutputDelegate, then, NumPips:int) {is=title K2Node_CustomEvent}
   > n6 = GetDynamicMaterial(self=Get ProgressBar())
   n7 = "To Float (Integer)"(InInt=@n5.NumPips)
   > SetScalarParameterValue(self=@n6, ParameterName=Segments, Value=@n7)
   > n8 = GetDynamicMaterial(self=Get ProgressBar_Delta())
   > SetScalarParameterValue(self=@n8, ParameterName=Segments, Value=@n7)
   n9 = SetPercent!() -> (OutputDelegate, then, Value:double=~) {is=title K2Node_CustomEvent}
   > n10 = GetDynamicMaterial(self=Get ProgressBar())
   > SetScalarParameterValue(self=@n10, ParameterName=Progress, Value=@n9.Value)
   > SetProgressWithAnimation!(execute, self, ProgressEnd:double=@n9.Value) -> (then)
   Event PreConstruct()
   > n11 = GetDynamicMaterial(self=Get ProgressBar())
   > SetVectorParameterValue(self=@n11, ParameterName=ColorA, Value=Get FillColorA() -> (FillColorA:Linear Color Structure))
   > SetVectorParameterValue(self=@n11, ParameterName=ColorB, Value=Get FillColorB() -> (FillColorB:Linear Color Structure))
   > SetScalarParameterValue(self=@n11, ParameterName=BackgroundOpacity, Value=0)
   > n12 = GetDynamicMaterial(self=Get ProgressBar_Delta())
   n13 = Get DeltaColor() -> (DeltaColor:Linear Color Structure)
   > SetVectorParameterValue(self=@n12, ParameterName=ColorA, Value=@n13)
   > SetVectorParameterValue(self=@n12, ParameterName=ColorB, Value=@n13)
   Branch(Condition=Get DifferentProgressColors() -> (DifferentProgressColors:bool))
   > SetHealthBarColorDependingHealth!(execute, self, CurrentPercentage:double=0) -> (then)
  ]}
 SetHealthBarColorDependingHealth {nodes [
   SetHealthBarColorDependingHealth!() -> (then, CurrentPercentage:double=~) {is=title K2Node_FunctionEntry}
   > n1 = GetDynamicMaterial(self=Get ProgressBar())
   n2 = Get CurrentPercentage() -> (CurrentPercentage:double)
   n3 = "float < float"(A=Get CurrentPercentage(), B=0.5)
   n4 = Select(
    Option 0:Linear Color Structure=Get FillColor1_100Percent() -> (FillColor1_100Percent:Linear Color Structure)
    Option 1:Linear Color Structure=Get FillColor1_75Percent() -> (FillColor1_75Percent:Linear Color Structure)
    Index="float < float"(A=Get CurrentPercentage(), B=0.75)
   ) -> (ReturnValue:Linear Color Structure)
   n5 = Get FillColor1_50Percent() -> (FillColor1_50Percent:Linear Color Structure)
   > SetVectorParameterValue(
    self=@n1
    ParameterName=ColorA
    Value=Select(
     Option 0=Select(Option 0=@n4, Option 1=@n5, Index=@n3)
     Option 1=Get FillColor1_25Percent() -> (FillColor1_25Percent:Linear Color Structure)
     Index="float < float"(A=@n2, B=0.25)
    )
   )
   n6 = Get CurrentPercentage()
   n7 = "float < float"(A=Get CurrentPercentage(), B=0.5)
   n8 = Select(
    Option 0=Get FillColor2_100Percent() -> (FillColor2_100Percent:Linear Color Structure)
    Option 1=Get FillColor2_75Percent() -> (FillColor2_75Percent:Linear Color Structure)
    Index="float < float"(A=Get CurrentPercentage(), B=0.75)
   )
   n9 = Get FillColor2_50Percent() -> (FillColor2_50Percent:Linear Color Structure)
   > SetVectorParameterValue(
    self=@n1
    ParameterName=ColorB
    Value=Select(
     Option 0=Select(Option 0=@n8, Option 1=@n9, Index=@n7)
     Option 1=Get FillColor2_25Percent() -> (FillColor2_25Percent:Linear Color Structure)
     Index="float < float"(A=@n6, B=0.25)
    )
   )
  ]}
 SetProgressWithAnimation {nodes [
   n1 = SetProgressWithAnimation!() -> (then, ProgressEnd:double=~) {is=title K2Node_FunctionEntry}
   > Set CurrentProgressValue(CurrentProgressValue=@n1)
   > n2 = Set Timer by Event(Delegate=Create Event(), Time=0.03, bLooping=true, InitialStartDelay=0.05)
   > Set Timer_DeltaAnimateProgressUpdate(Timer_DeltaAnimateProgressUpdate:Timer Handle Structure=@n2) -> (Output_Get:Timer Handle Structure)
  ]}
]
variables [
 CurrentProgressValue {type=number}
 DeltaColor {type=LinearColor, default=<rgba 1 0.6795 0.1119 1>}
 DifferentProgressColors {type=boolean}
 FillColor1_100Percent {type=LinearColor, default=<rgba 1 0.2387 0 0>}
 FillColor1_25Percent {type=LinearColor, default=<rgba 0.9047 0.0097 0.0097 1>}
 FillColor1_50Percent {type=LinearColor, default=<rgba 0.9047 0.0529 0.0097 1>}
 FillColor1_75Percent {type=LinearColor, default=<rgba 0.9047 0.1413 0.0097 1>}
 FillColor2_100Percent {type=LinearColor, default=<rgba 1 0.368 0 0>}
 FillColor2_25Percent {type=LinearColor, default=<rgba 0.9911 0.0144 0.0144 1>}
 FillColor2_50Percent {type=LinearColor, default=<rgba 0.9047 0.1356 0.0097 1>}
 FillColor2_75Percent {type=LinearColor, default=<rgba 0.9911 0.305 0.0144 1>}
 FillColorA {type=LinearColor, default=<rgba 0.9301 0.0865 0.0185 1>}
 FillColorB {type=LinearColor, default=<rgba 0.9301 0.2051 0.0168 1>}
 PreviousProgressValue {type=number}
 Timer_DeltaAnimateProgressUpdate {type=TimerHandle}
]
widgets {
 info {inheritedWidgetCount=0, parentClass=&UserWidget, rootWidgetClass=&SizeBox, widgetCount=6}
 widgets [
  BarSizeBox {widgetClassPath=&SizeBox, bIsVariable=true, widget=BarSizeBox, properties {heightOverride=24, widthOverride=450}}
  "[Overlay]" {
   widgetClassPath=&Overlay
   parent=BarSizeBox
   widget="[Overlay]"
   properties {visibility=HitTestInvisible}
   slot {class=SizeBoxSlot}
  }
  Image_CastShadow {
   widgetClassPath=&Image
   bIsVariable=true
   parent="[Overlay]"
   widget=Image_CastShadow
   properties {
    brush {
     bIsDynamicallyLoaded=false
     drawAs=Image
     imageType=NoImage
     mirroring=NoMirror
     resourceName=None
     resourceObject=/Game/Athena/HUD/Team/M_UI_BarBG_Pulse_Inst
     tiling=NoTile
     imageSize=<32 32>
     margin=<ltrb 0 0 0 0>
     outlineSettings {
      bUseBrushTransparency=false
      roundingType=HalfHeightRadius
      width=0
      color {colorUseRule=UseColor_Specified, specifiedColor=<rgba 0 0 0 0>}
      cornerRadii=<0 0 0 0>
     }
     tintColor {colorUseRule=UseColor_Specified, specifiedColor=<rgba 1 1 1 1>}
     uVRegion {bIsValid=false, max=<0 0>, min=<0 0>}
    }
    renderTransform {angle=0, scale=<1 1>, shear=<0 0>, translation=<5 5>}
   }
   slot {class=OverlaySlot, properties {horizontalAlignment=HAlign_Fill, verticalAlignment=VAlign_Fill}}
  }
  Image_BarBG {
   widgetClassPath=&Image
   bIsVariable=true
   parent="[Overlay]"
   widget=Image_BarBG
   properties {brush {
     bIsDynamicallyLoaded=false
     drawAs=NoDrawType
     imageType=NoImage
     mirroring=NoMirror
     resourceName=None
     resourceObject=null
     tiling=NoTile
     imageSize=<32 32>
     margin=<ltrb 0 0 0 0>
     outlineSettings {
      bUseBrushTransparency=false
      roundingType=HalfHeightRadius
      width=0
      color {colorUseRule=UseColor_Specified, specifiedColor=<rgba 0 0 0 0>}
      cornerRadii=<0 0 0 0>
     }
     tintColor {colorUseRule=UseColor_Specified, specifiedColor=<rgba 1 1 1 1>}
     uVRegion {bIsValid=false, max=<0 0>, min=<0 0>}
    }}
   slot {class=OverlaySlot, properties {horizontalAlignment=HAlign_Fill, verticalAlignment=VAlign_Fill}}
  }
  ProgressBar_Delta {
   widgetClassPath=&Image
   bIsVariable=true
   parent="[Overlay]"
   widget=ProgressBar_Delta
   properties {brush {
     bIsDynamicallyLoaded=false
     drawAs=Image
     imageType=FullColor
     mirroring=NoMirror
     resourceName=None
     resourceObject=/Game/Athena/HUD/Actors/Materials/M_UI_Status_Bar
     tiling=NoTile
     imageSize=<32 32>
     margin=<ltrb 0 0 0 0>
     outlineSettings {
      bUseBrushTransparency=false
      roundingType=HalfHeightRadius
      width=0
      color {colorUseRule=UseColor_Specified, specifiedColor=<rgba 0 0 0 0>}
      cornerRadii=<0 0 0 0>
     }
     tintColor {colorUseRule=UseColor_Specified, specifiedColor=<rgba 1 1 1 1>}
     uVRegion {bIsValid=false, max=<0 0>, min=<0 0>}
    }}
   slot {class=OverlaySlot, properties {horizontalAlignment=HAlign_Fill, verticalAlignment=VAlign_Fill}}
  }
  ProgressBar {
   widgetClassPath=&Image
   bIsVariable=true
   parent="[Overlay]"
   widget=ProgressBar
   properties {brush {
     bIsDynamicallyLoaded=false
     drawAs=Image
     imageType=FullColor
     mirroring=NoMirror
     resourceName=None
     resourceObject=&M_UI_Status_Bar
     tiling=NoTile
     imageSize=<12 12>
     margin=<ltrb 0 0 0 0>
     outlineSettings {
      bUseBrushTransparency=false
      roundingType=HalfHeightRadius
      width=0
      color {colorUseRule=UseColor_Specified, specifiedColor=<rgba 0 0 0 0>}
      cornerRadii=<0 0 0 0>
     }
     tintColor {colorUseRule=UseColor_Specified, specifiedColor=<rgba 1 1 1 1>}
     uVRegion {bIsValid=false, max=<0 0>, min=<0 0>}
    }}
   slot {class=OverlaySlot, properties {horizontalAlignment=HAlign_Fill, verticalAlignment=VAlign_Fill}}
  }
 ]
}
