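# NPC status HUD widget showing health and shield bars for an NPC pawn.
# Changed values update the bars and reveal the widget, as does pawn focus; a timer hides it after a configurable duration unless visibility is kept on.
# Playlist extension widgets receive the pawn and are placed beside or above the bars by slot type.
# A VisibilityChanged dispatcher announces each show and hide.
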WidgetBlueprint /Game/Athena/HUD/Actors/NPCStatusWidget
graphs [
 BP_HandlePlaylistExtensionWidgetCreated {nodes [
   Handle Playlist Extension Widget Created!() -> (then, Widget:User Widget Object Reference, SlotType:EUIExtensionSlot Enum=Invalid) {is=title K2Node_FunctionEntry}
   > n1 = Sequence()
   > n2 = Cast To NPCStatusWidgetExtensionBase!(execute, Object=Get Widget() -> (Widget:User Widget Object Reference)) -> (then, CastFailed, AsNPCStatus Widget Extension Base:NPCStatus Widget Extension Base Object Reference) {is=title K2Node_DynamicCast}
   > SetFortPawn!(
    execute
    self:NPCStatus Widget Extension Base Object Reference=@n2
    FortPawn:Fort Pawn Object Reference=Get FortPawn() -> (FortPawn:Fort Pawn Object Reference)
   ) -> (then)
   n3 = Branch(
    execute=@n1.then_1
    Condition="Equal (Enum)"(
     A:EUIExtensionSlot Enum=Get SlotType() -> (SlotType:EUIExtensionSlot Enum=Invalid)
     B:EUIExtensionSlot Enum=Get NPCStatusWidgetBarsPrefix() -> (NPCStatusWidgetBarsPrefix:EUIExtensionSlot Enum=Invalid)
    )
   )
   > n4 = AddChildToOverlay!(
    execute
    self:Overlay Object Reference=Get Overlay_BarsPrefix() -> (Overlay_BarsPrefix:Overlay Object Reference)
    Content:Widget Object Reference=Get Widget()
   ) -> (then, ReturnValue:Overlay Slot Object Reference)
   > SetVerticalAlignment(self=@n4, InVerticalAlignment=VAlign_Fill)
   > SetHorizontalAlignment(self=@n4, InHorizontalAlignment=HAlign_Fill)
   Branch(
    execute=@n3.else
    Condition="Equal (Enum)"(A=Get SlotType() -> (SlotType=Invalid), B=Get NPCStatusWidgetAboveBars() -> (NPCStatusWidgetAboveBars:EUIExtensionSlot Enum=Invalid))
   )
   > n5 = AddChildToVerticalBox(self=Get VerticalBox_BarExtensions() -> (VerticalBox_BarExtensions:Vertical Box Object Reference), Content=Get Widget())
   > SetVerticalAlignment(self:Vertical Box Slot Object Reference=@n5, InVerticalAlignment=VAlign_Fill)
   > SetHorizontalAlignment(self:Vertical Box Slot Object Reference=@n5, InHorizontalAlignment=HAlign_Fill)
  ]}
 EventGraph {nodes [
   n1 = EventUpdateStatus!() -> (OutputDelegate, then, health:double=~, shield:double=~, ShowShield:bool) {is=title K2Node_CustomEvent}
   n2 = Get CurrentHealthPercent() -> (CurrentHealthPercent:double)
   n3 = Get CurrentShieldPercent() -> (CurrentShieldPercent:double)
   > Branch(Condition=OR Boolean(A="Not Equal (Float)"(A=@n1.health, B=@n2), B="Not Equal (Float)"(A=@n1.shield, B=@n3)))
   > n4 = Set CurrentHealthPercent(CurrentHealthPercent:double=@n1.health) -> (Output_Get:double)
   > SetPercent!(execute, self:NPCStatus Widget Bar Object Reference=Get Bar_Health() -> (Bar_Health:NPCStatus Widget Bar Object Reference), Value:double=@n4) -> (then)
   > n5 = Branch(Condition=@n1.ShowShield)
   > n6 = Set CurrentShieldPercent(CurrentShieldPercent:double=@n1.shield) -> (Output_Get:double)
   n7 = Get Bar_Shield() -> (Bar_Shield:NPCStatus Widget Bar Object Reference)
   > SetPercent(self=@n7, Value=@n6)
   > n8 = SetVisibility(self=@n7, InVisibility=HitTestInvisible)
   SetVisibility(execute=[@n8.then, @n5.else], InVisibility=HitTestInvisible)
   > Call Visibility Changed!(execute:Exec, self:NPCStatus Widget Object Reference, bVisible:bool=true) -> (then:Exec) {is=title K2Node_CallDelegate}
   > n9 = Branch(Condition=Get KeepVisibilityOn() -> (KeepVisibilityOn:bool))
   n10 = Set Timer by Event(execute=@n9.else, Delegate=Create Event(), Time=Get StatusWidgetVisibleDuration() -> (StatusWidgetVisibleDuration:double), bLooping=false)
   > Set VisibiltyTimer(VisibiltyTimer:Timer Handle Structure=@n10) -> (Output_Get:Timer Handle Structure)
   EventHideStatus() {is=title K2Node_CustomEvent}
   > SetVisibility(InVisibility=Collapsed)
   > Call Visibility Changed()
   n11 = ChangeWidgetVisibleDuration!() -> (OutputDelegate, then, StatusWidgetVisibleDuration:double=~) {is=title K2Node_CustomEvent}
   > Set StatusWidgetVisibleDuration(StatusWidgetVisibleDuration:double=@n11.StatusWidgetVisibleDuration) -> (Output_Get:double)
   n12 = SetKeepVisibilityOn!() -> (OutputDelegate, then, KeepVisibilityOn:bool=~) {is=title K2Node_CustomEvent}
   > n13 = Set KeepVisibilityOn(KeepVisibilityOn:bool=@n12.KeepVisibilityOn) -> (Output_Get:bool)
   n14 = Get VisibiltyTimer() -> (VisibiltyTimer:Timer Handle Structure)
   > Branch(Condition=NAND Boolean(A=@n13, B=Is Valid Timer Handle!(Handle:Timer Handle Structure=@n14) -> (ReturnValue:bool)) {is=title K2Node_CommutativeAssociativeBinaryOperator})
   > EventHideStatus() {is=title K2Node_CallFunction}
   TryCallEventHideByTimer() {is=title K2Node_CustomEvent}
   > n15 = Branch(Condition=Get KeepVisibilityOn())
   EventHideStatus(execute=@n15.else)
   OnPawnFocused() {is=title K2Node_CustomEvent}
   > SetVisibility(InVisibility=HitTestInvisible)
   > Call Visibility Changed(bVisible=true)
   > n16 = Set Timer by Event(Delegate=Create Event(), Time=Get StatusWidgetVisibleDuration(), bLooping=false)
   > Set VisibiltyTimer(VisibiltyTimer=@n16)
   Event Construct()
   n17 = Get VisibiltyTimer()
   > n18 = Branch(Condition=OR Boolean(A=Is Valid Timer Handle(Handle=@n17), B=Get KeepVisibilityOn()))
   > SetPercent(self=Get Bar_Health(), Value=Get CurrentHealthPercent())
   > SetPercent(self=Get Bar_Shield(), Value=Get CurrentShieldPercent())
   SetVisibility(execute=@n18.else, InVisibility=Collapsed)
   > Call Visibility Changed()
  ]}
 VisibilityChanged {nodes [VisibilityChanged!() -> (then, bVisible:bool=~) {is=title K2Node_FunctionEntry}]}
]
variables [
 CurrentHealthPercent {type=number, default=1}
 CurrentShieldPercent {type=number}
 FortPawn {type=FortPawn, default=null}
 KeepVisibilityOn {type=boolean}
 NPCStatusWidgetAboveBars {type=EUIExtensionSlot, default=NPCStatusWidgetAboveBars}
 NPCStatusWidgetBarsPrefix {type=EUIExtensionSlot, default=NPCStatusWidgetBarsPrefix}
 StatusWidgetVisibleDuration {type=number, default=5}
 VisibilityChanged {}
 VisibiltyTimer {type=TimerHandle}
]
widgets {
 info {inheritedWidgetCount=0, parentClass=/Script/FortniteUI.NPCStatusWidgetBase, rootWidgetClass=&Overlay, widgetCount=9}
 widgets [
  Overlay_12 {widgetClassPath=&Overlay, widget=Overlay_12}
  HorizontalBox_0 {
   widgetClassPath=&HorizontalBox
   parent=Overlay_12
   widget=HorizontalBox_0
   properties {renderTransform {angle=0, scale=<0.75 0.75>, shear=<0 0>, translation=<0 -41>}}
   slot {class=OverlaySlot, properties {horizontalAlignment=HAlign_Fill, verticalAlignment=VAlign_Bottom}}
  }
  Overlay_BarsPrefix {
   widgetClassPath=&Overlay
   bIsVariable=true
   parent=HorizontalBox_0
   widget=Overlay_BarsPrefix
   slot {class=HorizontalBoxSlot, properties {padding=<ltrb 0 0 5 0>}}
  }
  VerticalBox_PlayerStatus {
   widgetClassPath=&VerticalBox
   parent=HorizontalBox_0
   widget=VerticalBox_PlayerStatus
   slot {class=HorizontalBoxSlot, properties {size {sizeRule=Fill, value=1}}}
  }
  VerticalBox_BarExtensions {
   widgetClassPath=&VerticalBox
   bIsVariable=true
   parent=VerticalBox_PlayerStatus
   widget=VerticalBox_BarExtensions
   slot {class=VerticalBoxSlot, properties {size {sizeRule=Fill, value=1}}}
  }
  SizeBox_0 {
   widgetClassPath=&SizeBox
   parent=VerticalBox_PlayerStatus
   widget=SizeBox_0
   properties {bOverride_HeightOverride=true, heightOverride=10}
   slot {class=VerticalBoxSlot, properties {verticalAlignment=VAlign_Top, padding=<ltrb 5 0 5 0>}}
  }
  Bar_Shield {
   widgetClassPath=&/NPCStatusWidgetBar
   bIsVariable=true
   parent=SizeBox_0
   widget=Bar_Shield
   properties {visibility=Collapsed, deltaColor=<rgba 0.521 0.9911 1 1>, fillColorA=<rgba 0.0009 0.3613 0.8714 1>, fillColorB=<rgba 0 0.7011 0.9647 1>}
   slot {class=SizeBoxSlot, properties {padding=<ltrb 0 0 0 4>}}
  }
  SizeBox_1 {
   widgetClassPath=&SizeBox
   parent=VerticalBox_PlayerStatus
   widget=SizeBox_1
   properties {bOverride_HeightOverride=true, heightOverride=20}
   slot {class=VerticalBoxSlot, properties {verticalAlignment=VAlign_Top}}
  }
  Bar_Health {widgetClassPath=&NPCStatusWidgetBar, bIsVariable=true, parent=SizeBox_1, widget=Bar_Health, slot {class=SizeBoxSlot}}
 ]
}
