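# HUD status widget for AI bosses: segmented health and optional shield bars between brackets, plus a name slot.
# EventUpdateStatus detects damage, derives pip counts from maximums above a threshold, and sets bars animated or instantly.
# Damage reveals the widget and starts a configurable visibility timer; EventHideStatus collapses it and clears the damage flag.
# Show, hide and Construct broadcast VisibilityChanged; Construct starts collapsed unless the timer is valid.
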WidgetBlueprint /Game/Athena/HUD/Actors/NPCStatusWidget_AIBoss
graphs [
 EventGraph {nodes [
   n1 = EventUpdateStatus!() -> (OutputDelegate, then, health:double=~, shield:double=~, maxHealth:double=~, maxShield:double=~, ShowShield:bool, InstantUpdate:bool=~) {is=title K2Node_CustomEvent}
   > SetIsDamaged!(execute, self, Health:double=@n1.health, Shield:double=@n1.shield) -> (then)
   > SetMaxShield!(execute, self, MaxShield:double=@n1.maxShield) -> (then)
   > SetMaxHealth!(execute, self, MaxHealth:double=@n1.maxHealth) -> (then)
   > SetNumPips!(
    execute
    self:NPCStatus Widget Bar Object Reference=Get Bar_Health() -> (Bar_Health:NPCStatus Widget Bar Object Reference)
    NumPips:int=Get NumHealthPips() -> (NumHealthPips:int=0)
   ) -> (then)
   > SetNumPips(self=Get Bar_Shield() -> (Bar_Shield:NPCStatus Widget Bar Object Reference), NumPips=Get NumShieldPips() -> (NumShieldPips:int=0))
   > n2 = Branch(Condition=@n1.InstantUpdate)
   > SetPercentInstant!(
    execute
    self:NPCStatus Widget Bar Object Reference=Get Bar_Health()
    Value:double=Get CurrentHealthPercent() -> (CurrentHealthPercent:double)
   ) -> (then)
   > n3 = SetPercentInstant(self=Get Bar_Shield(), Value=Get CurrentShieldPercent() -> (CurrentShieldPercent:double))
   n4 = Branch(execute=[@n3.then, @n8.then], Condition=@n1.ShowShield)
   n5 = Get Bar_Shield()
   > n6 = SetVisibility(self=@n5, InVisibility=HitTestInvisible)
   Branch(execute=[@n6.then, @n7.then], Condition=Get IsDamaged() -> (IsDamaged:bool))
   > EventShowStatus()
   n7 = SetVisibility(execute=@n4.else, self=@n5, InVisibility=Collapsed)
   SetPercent!(execute=@n2.else, self:NPCStatus Widget Bar Object Reference=Get Bar_Health(), Value:double=Get CurrentHealthPercent()) -> (then)
   > n8 = SetPercent(self=Get Bar_Shield(), Value=Get CurrentShieldPercent())
   EventHideStatus() {is=title K2Node_CustomEvent}
   > SetVisibility(InVisibility=Collapsed)
   > Call Visibility Changed!(execute:Exec, self:NPCStatus Widget AIBoss Object Reference, bVisible:bool) -> (then:Exec) {is=title K2Node_CallDelegate}
   > Set IsDamaged(IsDamaged:bool=false) -> (Output_Get:bool)
   n9 = ChangeWidgetVisibleDuration!() -> (OutputDelegate, then, StatusWidgetVisibleDuration:double=~) {is=title K2Node_CustomEvent}
   > Branch(Condition="Not Equal (Float)"(A=@n9.StatusWidgetVisibleDuration, B=Get StatusWidgetVisibleDuration() -> (StatusWidgetVisibleDuration:double)))
   > n10 = Set StatusWidgetVisibleDuration(StatusWidgetVisibleDuration:double=@n9.StatusWidgetVisibleDuration) -> (Output_Get:double)
   > n11 = Branch(Condition="float > float"(A=@n10))
   > n12 = Set Timer by Event(Delegate=Create Event(), Time=Get StatusWidgetVisibleDuration(), bLooping=false)
   > Set VisibiltyTimer(VisibiltyTimer:Timer Handle Structure=@n12) -> (Output_Get:Timer Handle Structure)
   Clear and Invalidate Timer by Handle(execute=@n11.else, Handle=Get VisibiltyTimer() -> (VisibiltyTimer:Timer Handle Structure))
   EventShowStatus() {is=title K2Node_CustomEvent}
   > SetVisibility(InVisibility=HitTestInvisible)
   > Call Visibility Changed(bVisible=true)
   > Branch(Condition="float > float"(A=Get StatusWidgetVisibleDuration()))
   > n13 = Set Timer by Event(Delegate=Create Event(), Time=Get StatusWidgetVisibleDuration(), bLooping=false)
   > Set VisibiltyTimer(VisibiltyTimer=@n13)
   Event Construct()
   > n14 = Branch(Condition=Is Valid Timer Handle!(Handle:Timer Handle Structure=Get VisibiltyTimer()) -> (ReturnValue:bool))
   > SetPercentInstant(self=Get Bar_Health(), Value=Get CurrentHealthPercent())
   > SetPercentInstant(self=Get Bar_Shield(), Value=Get CurrentShieldPercent())
   SetVisibility(execute=@n14.else, InVisibility=Collapsed)
   > Call Visibility Changed()
  ]}
 SetIsDamaged {nodes [
   n1 = SetIsDamaged!() -> (then, Health:double=~, Shield:double=~) {is=title K2Node_FunctionEntry}
   n2 = AND Boolean(A="Equal (Float)"(A=Get CurrentHealthPercent()), B="Equal (Float)"(A=Get CurrentShieldPercent()))
   n3 = OR Boolean(A="float > float"(A=@n1), B="float > float"(A=@n1.Shield))
   n4 = Get CurrentHealthPercent()
   n5 = Get CurrentShieldPercent()
   > Set IsDamaged(IsDamaged=OR Boolean!(A="float < float"(A=@n1, B=@n4), B="float < float"(A=@n1.Shield, B=@n5), C:bool=AND Boolean(A=@n2, B=@n3)) -> (ReturnValue))
   > Set CurrentHealthPercent(CurrentHealthPercent:double=Get Health() -> (Health:double)) -> (Output_Get:double)
   > Set CurrentShieldPercent(CurrentShieldPercent:double=Get Shield() -> (Shield:double)) -> (Output_Get:double)
  ]}
 SetMaxHealth {nodes [
   n1 = SetMaxHealth!() -> (then, MaxHealth:double=~) {is=title K2Node_FunctionEntry}
   > Branch(Condition="float >= float"(A=@n1, B=Get MinimumPipAmount() -> (MinimumPipAmount:double)))
   > n2 = Branch(Condition="Nearly Equal (Float)"(A=@n1, B=Get Max Health() -> (Max Health:double), ErrorTolerance=0.01))
   n3 = Set Max Health(execute=@n2.else, Max Health:double=@n1) -> (Output_Get:double)
   n4 = float / float(A=@n3, B=Get PipAmount() -> (PipAmount:double))
   > Set NumHealthPips(NumHealthPips:int="To Integer (Integer64)"!(InInt:Integer64=Floor to Integer64!(A:double=@n4) -> (ReturnValue:Integer64=0)) -> (ReturnValue:int=0)) -> (Output_Get:int=0)
  ]}
 SetMaxShield {nodes [
   n1 = SetMaxShield!() -> (then, MaxShield:double=~) {is=title K2Node_FunctionEntry}
   > Branch(Condition="float >= float"(A=@n1, B=Get MinimumPipAmount()))
   > n2 = Branch(Condition="Nearly Equal (Float)"(A=@n1, B=Get MaxShield() -> (MaxShield:double), ErrorTolerance=0.01))
   n3 = Set MaxShield(execute=@n2.else, MaxShield:double=@n1) -> (Output_Get:double)
   n4 = float / float(A=@n3, B=Get PipAmount())
   > Set NumShieldPips(NumShieldPips:int="To Integer (Integer64)"(InInt=Floor to Integer64(A=@n4) -> (ReturnValue=0)) -> (ReturnValue=0)) -> (Output_Get:int=0)
  ]}
 VisibilityChanged {nodes [VisibilityChanged!() -> (then, bVisible:bool=~) {is=title K2Node_FunctionEntry}]}
]
variables [
 CurrentHealthPercent {type=number}
 CurrentShieldPercent {type=number}
 Font Outline Settings {
  type=FontOutlineSettings
  default {
   bApplyOutlineToDropShadows=false
   bMiteredCorners=false
   bSeparateFillAlpha=false
   outlineMaterial=null
   outlineSize=0
   outlineColor=<rgba 0 0 0 1>
  }
 }
 IsDamaged {type=boolean}
 Max Health {type=number}
 MaxShield {type=number}
 MinimumPipAmount {type=number, default=300}
 NumHealthPips {type=integer}
 NumShieldPips {type=integer}
 PipAmount {type=number, default=100}
 StatusWidgetVisibleDuration {type=number, default=5}
 VisibilityChanged {}
 VisibiltyTimer {type=TimerHandle}
]
widgets {
 info {inheritedWidgetCount=1, parentClass=&FortHUDElementWidget, rootWidgetClass=&Overlay, widgetCount=14}
 widgets [
  Overlay_0 {widgetClassPath=&Overlay, widget=Overlay_0}
  Background {
   widgetClassPath=&Image
   bIsVariable=true
   parent=Overlay_0
   widget=Background
   properties {
    visibility=Collapsed
    brush {
     bIsDynamicallyLoaded=false
     drawAs=Image
     imageType=NoImage
     mirroring=NoMirror
     resourceName=None
     resourceObject=null
     tiling=NoTile
     imageSize=<161 18>
     margin=<ltrb 0 0 0 0>
     outlineSettings {
      bUseBrushTransparency=false
      roundingType=HalfHeightRadius
      width=0
      color {colorUseRule=UseColor_Specified, specifiedColor=<rgba 0 0 0 0>}
      cornerRadii=<0 0 0 0>
     }
     tintColor {colorUseRule=UseColor_Specified, specifiedColor=<rgba 1 1 1 1>}
     uVRegion {bIsValid=false, max=<0 0>, min=<0 0>}
    }
    colorAndOpacity=<rgba 0 0 0 0.32>
   }
   slot {class=OverlaySlot, properties {horizontalAlignment=HAlign_Center, verticalAlignment=VAlign_Center}}
  }
  HorizontalBox_0 {
   widgetClassPath=&HorizontalBox
   parent=Overlay_0
   widget=HorizontalBox_0
   slot {class=OverlaySlot, properties {horizontalAlignment=HAlign_Center, verticalAlignment=VAlign_Center}}
  }
  VerticalBox_PlayerStatus {
   widgetClassPath=&VerticalBox
   parent=HorizontalBox_0
   widget=VerticalBox_PlayerStatus
   properties {renderTransform {angle=0, scale=<0.75 0.75>, shear=<0 0>, translation=<0 0>}}
   slot {class=HorizontalBoxSlot}
  }
  NamedSlot_Name {
   widgetClassPath=&NamedSlot
   bIsVariable=true
   parent=VerticalBox_PlayerStatus
   widget=NamedSlot_Name
   properties {visibility=HitTestInvisible}
   slot {class=VerticalBoxSlot}
  }
  HorizontalBox_Status {
   widgetClassPath=&HorizontalBox
   parent=VerticalBox_PlayerStatus
   widget=HorizontalBox_Status
   slot {class=VerticalBoxSlot, properties {horizontalAlignment=HAlign_Center, verticalAlignment=VAlign_Center}}
  }
  BracketL {
   widgetClassPath=&Image
   parent=HorizontalBox_Status
   widget=BracketL
   properties {
    brush {
     bIsDynamicallyLoaded=false
     drawAs=Image
     imageType=FullColor
     mirroring=NoMirror
     resourceName=None
     resourceObject=/Game/Athena/HUD/Actors/Materials/MI_UI_Status_Bracket
     tiling=NoTile
     imageSize=<21 2>
     margin=<ltrb 0 0 0 0>
     outlineSettings {
      bUseBrushTransparency=false
      roundingType=HalfHeightRadius
      width=0
      color {colorUseRule=UseColor_Specified, specifiedColor=<rgba 0 0 0 0>}
      cornerRadii=<0 0 0 0>
     }
     tintColor {colorUseRule=UseColor_Specified, specifiedColor=<rgba 1 1 1 1>}
     uVRegion {bIsValid=false, max=<0 0>, min=<0 0>}
    }
    renderTransform {angle=0, scale=<2 2>, shear=<0 0>, translation=<0 0>}
   }
   slot {class=HorizontalBoxSlot, properties {padding=<ltrb 0 0 -12 -4>}}
  }
  VerticalBox_0 {
   widgetClassPath=&VerticalBox
   parent=HorizontalBox_Status
   widget=VerticalBox_0
   slot {class=HorizontalBoxSlot, properties {horizontalAlignment=HAlign_Center, verticalAlignment=VAlign_Center}}
  }
  SizeBox_0 {
   widgetClassPath=&SizeBox
   parent=VerticalBox_0
   widget=SizeBox_0
   properties {bOverride_HeightOverride=true, bOverride_MinDesiredWidth=true, heightOverride=9, minDesiredWidth=215, widthOverride=99.468}
   slot {class=VerticalBoxSlot, properties {verticalAlignment=VAlign_Center, padding=<ltrb 3 0 3 0>}}
  }
  Bar_Shield {
   widgetClassPath=&/NPCStatusWidgetBar
   bIsVariable=true
   parent=SizeBox_0
   widget=Bar_Shield
   properties {visibility=Collapsed, deltaColor=<rgba 0.521 0.9911 1 1>, fillColorA=<rgba 0.0009 0.3613 0.8714 1>, fillColorB=<rgba 0 0.7011 0.9647 1>}
   slot {class=SizeBoxSlot, properties {padding=<ltrb 0 0 0 2>}}
  }
  SizeBox_Health {
   widgetClassPath=&SizeBox
   bIsVariable=true
   parent=VerticalBox_0
   widget=SizeBox_Health
   properties {bOverride_HeightOverride=true, bOverride_MinDesiredWidth=true, heightOverride=12, minDesiredWidth=215}
   slot {class=VerticalBoxSlot, properties {verticalAlignment=VAlign_Center}}
  }
  Bar_Health {
   widgetClassPath=&NPCStatusWidgetBar
   bIsVariable=true
   parent=SizeBox_Health
   widget=Bar_Health
   slot {class=SizeBoxSlot}
  }
  BracketR {
   widgetClassPath=&Image
   parent=HorizontalBox_Status
   widget=BracketR
   properties {
    brush {
     bIsDynamicallyLoaded=false
     drawAs=Image
     imageType=FullColor
     mirroring=NoMirror
     resourceName=None
     resourceObject=&MI_UI_Status_Bracket
     tiling=NoTile
     imageSize=<21 2>
     margin=<ltrb 0 0 0 0>
     outlineSettings {
      bUseBrushTransparency=false
      roundingType=HalfHeightRadius
      width=0
      color {colorUseRule=UseColor_Specified, specifiedColor=<rgba 0 0 0 0>}
      cornerRadii=<0 0 0 0>
     }
     tintColor {colorUseRule=UseColor_Specified, specifiedColor=<rgba 1 1 1 1>}
     uVRegion {bIsValid=false, max=<0 0>, min=<0 0>}
    }
    renderTransform {angle=0, scale=<-2 2>, shear=<0 0>, translation=<0 0>}
   }
   slot {class=HorizontalBoxSlot, properties {padding=<ltrb -12 0 0 -4>}}
  }
  TutorialHighlight {widgetClassPath=&TutorialHighlight}
 ]
}
