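# NPC status widget variant of NPCStatusWidget_AIBoss that adds a centered name label for the NPC.
# SetPlayerName writes the incoming text to Text_Name, then shows the label if b_ShouldShowName evaluates to exactly 1, otherwise hides it.
# That flag is a ScalableFloat tied to a DT_NPCStatusWidget curve table row, so name display is data-driven.
# The event graph is empty, and BossPawnName is declared but unused here.
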WidgetBlueprint /Game/Athena/HUD/Actors/WBP_NPCStatusWidget_WithName
graphs [
 EventGraph {links [], nodes []}
 SetPlayerName {nodes [
   n1 = SetPlayerName!() -> (then, InPlayerName:Text) {is=title K2Node_FunctionEntry}
   n2 = Get Text_Name() -> (Text_Name:Common Text Object Reference)
   > Set Text(Text=@n1, self=@n2)
   n3 = Get b_ShouldShowName() -> (b_ShouldShowName:Scalable Float Structure)
   > n4 = Branch(Condition="Equal (Float)"(A=GetValueAtLevel!(Input:"Scalable Float Structure (by ref)"=@n3, Level:float=0) -> (ReturnValue:float), B=1))
   > SetVisibility(self=@n2, InVisibility=Visible)
   SetVisibility(execute=@n4.else, self=@n2, InVisibility=Hidden)
  ]}
]
variables [
 b_ShouldShowName {
  type=ScalableFloat
  default {registryType=None, value=1, curve {curveTable=&/DT_NPCStatusWidget, rowName=b_ShouldShowName}}
 }
 BossPawnName {type=string, default=""}
]
widgets {
 info {
  inheritedWidgetCount=0
  namedSlotCount=1
  parentClass=&/NPCStatusWidget_AIBoss.NPCStatusWidget_AIBoss_C
  rootWidgetClass=null
  widgetCount=3
 }
 widgets [
  Overlay_Name {widgetClassPath=&Overlay, widget=Overlay_Name}
  HorizontalBox_1 {
   widgetClassPath=&HorizontalBox
   parent=Overlay_Name
   widget=HorizontalBox_1
   slot {class=OverlaySlot, properties {horizontalAlignment=HAlign_Center, verticalAlignment=VAlign_Center, padding=<ltrb 0 0 0 10>}}
  }
  Text_Name {
   widgetClassPath=&CommonTextBlock
   bIsVariable=true
   parent=HorizontalBox_1
   widget=Text_Name
   properties {
    mobileFontSizeMultiplier=1.4
    scrollStyle=&TextScrollStyle-Base_C
    text=PlayerName_default
    font {
     bForceMonospaced=false
     bMaterialIsStencil=false
     fontMaterial=null
     fontObject=&Burbank
     letterSpacing=0
     monospacedWidth=1
     size=14
     skewAmount=0
     typefaceFontName=BurbankRegular-Black
     outlineSettings {
      bApplyOutlineToDropShadows=true
      bMiteredCorners=false
      bSeparateFillAlpha=false
      outlineMaterial=null
      outlineSize=1
      outlineColor=<rgba 0 0 0 0.75>
     }
    }
    shadowColorAndOpacity=<rgba 0 0 0 1>
   }
   slot {class=HorizontalBoxSlot}
  }
 ]
}
